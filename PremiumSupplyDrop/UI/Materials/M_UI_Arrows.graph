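# UI material for the Premium Supply Drop that draws procedural, horizontally scrolling orange arrows with alpha-composite blending.
# UVs tiled three times in X pan by the Speed parameter, then pass through rotation and a mirrored symmetry function, adjustable via Offset and Scale, into a box SDF function to form the repeating stripes.
# A box with one trapezoid subtracted and another added at the ends clips the pattern, which is then tinted by a fixed color.
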Material /PremiumSupplyDrop/UI/Materials/M_UI_Arrows
graph {
 properties {bAutomaticallySetUsageInEditor=false, blendMode=BLEND_AlphaComposite, materialDomain=MD_UI, usedShadingModels=Default Lit}
 nodes [
  n1 = Constant() {r=3}
  n2 = TextureCoordinate()
  n3 = ComponentMask(None=@n2) {r=true}
  n4 = Panner(
   Coordinate=AppendVector(A=Multiply(A=@n1, B=@n3), B=ComponentMask(None=@n2) {g=true})
   Speed=AppendVector(A=ScalarParameter() {defaultValue=-1, parameterName=Speed}, B=Constant())
  ) {speedX=0.1}
  n5 = Constant() {r=0.25}
  n6 = VectorParameter() {parameterName=Offset, defaultValue=<rgba 0.75 0.28 0 1>}
  n7 = VectorParameter() {parameterName=Scale, defaultValue=<rgba 6 1 0 1>}
  n8 = MaterialFunctionCall(
   UVs=MaterialFunctionCall!(
    UV=MaterialFunctionCall(UVs=@n4, "Rotation Angle (0-1)"=@n5) {materialFunction=&CustomRotator}.Rotated Values
    Offset=ComponentMask(None=@n6.RGB) {g=true, r=true}
    Rotation=Constant() {r=-0.03}
    Scale=ComponentMask(None=@n7.RGB) {g=true, r=true}
    MirrorY=StaticBool() {value=true}
   ) {materialFunction=/FortUILibrary/MaterialFunctions/Basics/MF_UI_Symmetry_2x1_DEPRECATED}.Result
   GlowMax=Constant() {r=0.02}
  ) {materialFunction=&MF_UI_Box}
  n9 = TextureCoordinate()
  n10 = Constant2Vector() {r=0.025}
  n11 = MaterialFunctionCall(UVs=Add(A=TextureCoordinate(), B=Constant2Vector() {r=0.475}), "Rotation Angle (0-1)"=Constant() {r=0.75}) {materialFunction=&CustomRotator}
  n12 = Constant()
  n13 = Constant() {r=1}
  n14 = Constant() {r=0.05}
  n15 = Subtract(
   A=MaterialFunctionCall(UVs=Add(A=@n9, B=@n10), Size=Constant2Vector() {g=1, r=0.945}) {materialFunction=&MF_UI_Box}.Fill
   B=Saturate(None=MaterialFunctionCall!(UVs=@n11.Rotated Values, TopWidth=@n12, BottomWidth=@n13, Height=@n14, OutlineThickness, OutlineAlignment, GlowMin, GlowMax) {materialFunction=/FortUILibrary/MaterialFunctions/Shapes/MF_UI_IsoscelesTrapezoid}.Fill)
  )
  n16 = Add(A=TextureCoordinate(), B=Constant2Vector() {r=-0.47})
  n17 = Constant() {r=0.75}
  n18 = Saturate(None=MaterialFunctionCall(
    UVs=MaterialFunctionCall(UVs=@n16, "Rotation Angle (0-1)"=@n17) {materialFunction=&CustomRotator}.Rotated Values
    TopWidth=Constant()
    BottomWidth=Constant() {r=1}
    Height=Constant() {r=0.05}
   ) {materialFunction=&MF_UI_IsoscelesTrapezoid}.Fill)
  n19 = Saturate(None=OneMinus(None=Add(A=@n15, B=@n18)))
  Multiply(A=Saturate(None=Subtract(A=@n8.Fill, B=@n19)), B=Constant4Vector() {constant=<rgba 1 0.1413 0 1>}.RGBA)
 ]
}
