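# UI material that draws a supply drop pod icon over an orange-tinted decorative background, alpha-composited.
# Named reroutes hold each layer: SDF-antialiased center icon, box-masked line texture, clipped octagon with gradient glow, two rounded squares, faint procedural grid, soft box mask and time-stepped noise.
# Two parallel sums combine these layers, one tinted for color and one untinted for opacity.
# Scalar and vector parameters expose layer opacities, icon scale, glow and line shaping.
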Material /PremiumSupplyDrop/UI/Materials/M_UI_IconWithBackground
graph {
 properties {bAutomaticallySetUsageInEditor=false, blendMode=BLEND_AlphaComposite, materialDomain=MD_UI, usedShadingModels=Default Lit}
 nodes [
  n1 = NamedRerouteUsage()
  n2 = NamedRerouteUsage()
  n3 = NamedRerouteUsage()
  n4 = Multiply(A=NamedRerouteUsage(), B=NamedRerouteUsage())
  n5 = Add(
   A=Add(A=Multiply(A=@n1, B=@n2), B=Add(A=@n3, B=@n4))
   B=Multiply(A=NamedRerouteUsage(), B=ScalarParameter() {defaultValue=0.7, parameterName=Square 2 Opacity})
  )
  n6 = Multiply(A=NamedRerouteUsage(), B=ScalarParameter() {defaultValue=0.7, parameterName=Square 1 Opacity})
  n7 = Multiply(A=NamedRerouteUsage(), B=ScalarParameter() {defaultValue=1, parameterName=Icon Opacity})
  Add(A=Add(A=Add(A=@n5, B=@n6), B=Saturate(None=@n7)), B=Multiply(A=NamedRerouteUsage()) {constB=0.5})
  n8 = NamedRerouteUsage()
  n9 = NamedRerouteUsage()
  n10 = Add(A=NamedRerouteUsage(), B=Multiply(A=NamedRerouteUsage(), B=NamedRerouteUsage()))
  n11 = Constant4Vector() {constant=<rgba 1 0.1413 0 1>}
  n12 = NamedRerouteUsage()
  n13 = Constant4Vector() {constant=<rgba 1 0.1413 0 1>}
  n14 = Add(
   A=Add(A=Multiply(A=@n8, B=@n9), B=Multiply(A=@n10, B=@n11.RGBA))
   B=Multiply(A=Multiply(A=@n12, B=@n13.RGBA), B=ScalarParameter() {defaultValue=0.7, parameterName=Square 2 Opacity})
  )
  n15 = Multiply(
   A=Multiply(A=NamedRerouteUsage(), B=Constant4Vector() {constant=<rgba 1 0.1413 0 1>}.RGBA)
   B=ScalarParameter() {defaultValue=0.7, parameterName=Square 1 Opacity}
  )
  n16 = Multiply(A=NamedRerouteUsage(), B=ScalarParameter() {defaultValue=1, parameterName=Icon Opacity})
  Add(A=Add(A=Add(A=@n14, B=@n15), B=Saturate(None=@n16)), B=Multiply(A=NamedRerouteUsage()) {constB=0.2})
  n17 = ScalarParameter() {defaultValue=0.1, parameterName=Outer Line Mask UV X}
  n18 = ScalarParameter() {defaultValue=0.3, parameterName=Outer Line Mask UV Y}
  n19 = MaterialFunctionCall(
   UVs=Add(A=TextureCoordinate(), B=AppendVector(A=@n17, B=@n18))
   Size=VectorParameter() {parameterName=Outer Line Mask Size, defaultValue=<rgba 1 1.4 0 1>}.RGB
  ) {materialFunction=&MF_UI_Box}
  n20 = Add(A=TextureCoordinate(), B=Constant2Vector() {r=-0.1})
  n21 = Constant() {r=0.9}
  n22 = MaterialFunctionCall(
   DistanceField=TextureSample(UVs=MaterialFunctionCall(UVs=@n20, Texture Scale=@n21) {materialFunction=&ScaleUVsByCenter}.UVs) {texture=/PremiumSupplyDrop/UI/Textures/T_UI_DropPod_LinesA}.R
   Threshold=Constant() {r=0.5}
  ) {materialFunction=&MF_UI_AntiAliasing}
  NamedRerouteDeclaration(None=Saturate(None=Subtract(A=@n19.Fill, B=@n22.Inner))) {name=Lines, nodeColor=<rgba 1 0.4453 0 1>}
  n23 = MaterialFunctionCall(UVs=Add(A=TextureCoordinate(), B=Constant2Vector() {g=-0.05}), Size=Constant2Vector() {g=0.5, r=0.56}) {materialFunction=&MF_UI_Box}
  n24 = TextureCoordinate()
  n25 = Constant2Vector() {g=0.158}
  n26 = MaterialFunctionCall!(
   UVs=MaterialFunctionCall(UVs=Add(A=@n24, B=@n25), "Rotation Angle (0-1)"=Constant() {r=0.063}) {materialFunction=&CustomRotator}.Rotated Values
   Size=Constant() {r=0.965}
   Sides=Constant() {r=8}
   InnerAngle
   OutlineThickness=ScalarParameter() {defaultValue=0.015, parameterName=Bottom Line Thickness}
   OutlineAlignment
   GlowMin
   GlowMax
  ) {materialFunction=/FortUILibrary/MaterialFunctions/Shapes/MF_UI_RegularPolygon}
  n27 = MaterialFunctionCall(
   GradientStart=ScalarParameter() {defaultValue=0.3, parameterName=Background Glow}
   GradientEnd=Constant() {r=1}
   RotationAngle=Constant() {r=0.75}
  ) {materialFunction=&MF_UI_LinearGradient}
  n28 = Add(A=Multiply(A=@n23.Fill, B=@n26.Outline), B=Multiply(A=Power(Base=@n27.Result), B=Multiply(A=@n26.Fill, B=@n23.Fill)))
  n29 = ScalarParameter() {defaultValue=1, parameterName=Background Glow Intensity}
  NamedRerouteDeclaration(None=Saturate(None=Multiply(A=@n28, B=@n29))) {name=Background, nodeColor=<rgba 0 1 0.1328 1>}
  n30 = TextureCoordinate()
  n31 = Constant2Vector() {g=0.37, r=0.22}
  NamedRerouteDeclaration(None=MaterialFunctionCall(UVs=Add(A=@n30, B=@n31), Size=Constant2Vector() {g=0.06, r=0.06}, CornerRoundness=Constant() {r=0.02}) {materialFunction=&MF_UI_Box}.Fill) {name=Square 01, nodeColor=<rgba 0 1 0.0391 1>}
  n32 = TextureCoordinate()
  n33 = Constant2Vector() {g=-0.371, r=-0.015}
  NamedRerouteDeclaration(None=MaterialFunctionCall(UVs=Add(A=@n32, B=@n33), Size=Constant2Vector() {g=0.06, r=0.06}, CornerRoundness=Constant() {r=0.02}) {materialFunction=&MF_UI_Box}.Fill) {name=Square 02, nodeColor=<rgba 0 1 0.0391 1>}
  n34 = TextureCoordinate()
  n35 = Subtract(A=ComponentMask(None=@n34) {r=true}) {constB=0.5}
  n36 = SmoothStep(Value=Frac(None=Multiply(A=@n35) {constB=7})) {constMax=0.99, constMin=0.95}
  n37 = Subtract(A=ComponentMask(None=@n34) {g=true}) {constB=0.5}
  n38 = SmoothStep(Value=Frac(None=Multiply(A=@n37) {constB=7})) {constMax=0.99, constMin=0.95}
  NamedRerouteDeclaration(None=Multiply(A=Add(A=@n36, B=@n38)) {constB=0.02}) {name=GridBg, nodeColor=<rgba 1 0 0.0703 1>}
  n39 = Constant2Vector() {r=1.4}
  n40 = Constant() {r=-0.5}
  NamedRerouteDeclaration(None=SmoothStep(Value=MaterialFunctionCall!(UVs=@n34, Size=@n39, CornerRoundness=@n40, KeepAspectRatio, PixelSize) {materialFunction=/FortUILibrary/MaterialFunctions/Shapes/SDFs/MF_SDF_Box}.SDF) {constMax=0.7, constMin=1}) {name=BoxMask, nodeColor=<rgba 1 0.4453 0 1>}
  n41 = Frac(None=Time())
  n42 = Add(
   A=MaterialFunctionCall(UV=TextureCoordinate() {uTiling=3, vTiling=3}) {materialFunction=&MF_UI_UVScaling}.Scaled to Fit
   B=MaterialFunctionCall!("Param (0-1)"=Multiply(A=@n41) {constB=2}, Tiles=Constant() {r=5}, Apply Offset) {materialFunction=/Game/Environments/Papaya/Materials/MaterialFunctions/MF_Papaya_Quantize_Scalar}.Result
  )
  NamedRerouteDeclaration(None=Multiply(A=TextureSample(UVs=@n42) {samplerType=SAMPLERTYPE_Normal, texture=/Game/Global/Textures/Default/Noise/64x64Noise}.R) {constB=0.05}) {name=Noise, nodeColor=<rgba 1 0 0.0938 1>}
  n43 = TextureCoordinate()
  n44 = Constant2Vector() {g=0.02}
  n45 = TextureSample(UVs=MaterialFunctionCall(UVs=Add(A=@n43, B=@n44), Texture Scale=ScalarParameter() {defaultValue=0.55, parameterName=IconScale}) {materialFunction=&ScaleUVsByCenter}.UVs) {texture=/DropPodParentGameplay/Assets/UI/Icons/T_UI_DropPod_Icon}
  n46 = Constant() {r=0.5}
  NamedRerouteDeclaration(None=Multiply(A=MaterialFunctionCall(DistanceField=@n45.R, Threshold=@n46) {materialFunction=&MF_UI_AntiAliasing}.Outer)) {name=CenterIcon, nodeColor=<rgba 1 0 0.0234 1>}
 ]
}
