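# UI material with alpha-composite blending that draws a procedural text background panel: rounded fill, partial outline and small triangle accent.
# Shapes come from MF_UI_Box on offset 9-slice UVs, without textures.
# The stroke is an outline trimmed by an offset box and a parameterized mask; the triangle is a square cut by a rotated box.
# Named reroutes feed two layer sums, one untinted and one color-tinted, with background and triangle opacity parameters.
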Material /PremiumSupplyDrop/UI/Materials/M_UI_TextBackground
graph {
 properties {bAutomaticallySetUsageInEditor=false, blendMode=BLEND_AlphaComposite, materialDomain=MD_UI, usedShadingModels=Default Lit}
 nodes [
  n1 = Add(
   A=MaterialFunctionCall!(UVs) {materialFunction=/FortUILibrary/MaterialFunctions/Basics/MF_UI_9SliceUVs}.UVs
   B=Constant2Vector() {g=0.009, r=-0.009}
  )
  n2 = Constant2Vector() {g=0.41, r=0.96}
  n3 = Constant4Vector() {constant=<rgba 0 0.05 0 0>}
  n4 = Multiply(
   A=NamedRerouteDeclaration(None=MaterialFunctionCall(UVs=@n1, Size=@n2, CornerRoundness=@n3.RGBA) {materialFunction=&MF_UI_Box}.Fill) {name=Background, nodeColor=<rgba 0 1 0.9766 1>}
   B=ScalarParameter() {defaultValue=0.5, parameterName=Background Opacity}
  ) {constB=0.5}
  n5 = MaterialFunctionCall() {materialFunction=&MF_UI_9SliceUVs}
  n6 = Constant2Vector()
  n7 = Constant4Vector() {constant=<rgba 0 0.05 0 0>}
  n8 = MaterialFunctionCall() {materialFunction=&MF_UI_9SliceUVs}
  n9 = Constant2Vector() {g=0.054, r=-0.015}
  n10 = Subtract(
   A=MaterialFunctionCall(UVs=Add(A=@n5.UVs, B=@n6), Size=Constant2Vector() {g=0.4, r=0.95}, CornerRoundness=@n7.RGBA, OutlineThickness=Constant() {r=0.03}) {materialFunction=&MF_UI_Box}.Outline
   B=MaterialFunctionCall(UVs=Add(A=@n8.UVs, B=@n9), Size=Constant2Vector() {g=0.5, r=0.97}, CornerRoundness=@n7.RGBA) {materialFunction=&MF_UI_Box}.Fill
  )
  n11 = MaterialFunctionCall() {materialFunction=&MF_UI_9SliceUVs}
  n12 = AppendVector(A=ScalarParameter() {parameterName=Line Shape Mask UV X}, B=ScalarParameter() {parameterName=Line Shape Mask UV Y})
  n13 = OneMinus(None=MaterialFunctionCall(UVs=Add(A=@n11.UVs, B=@n12), Size=VectorParameter() {parameterName=Line Shape Mask Size, defaultValue=<rgba 1 1 0 1>}.RGB) {materialFunction=&MF_UI_Box}.Fill)
  n14 = Multiply(A=NamedRerouteDeclaration(None=Subtract(A=@n10, B=@n13)) {name=Stroke, nodeColor=<rgba 1 0.6563 0 1>}) {constB=0.5}
  Add(
   A=Add(A=Saturate(None=@n4), B=Saturate(None=@n14))
   B=Multiply(A=NamedRerouteUsage(), B=ScalarParameter() {defaultValue=0.7, parameterName=Triangle Opacity})
  )
  n15 = NamedRerouteUsage()
  n16 = Constant4Vector() {constant=<rgba 0 0 0 1>}
  n17 = Multiply(B=Multiply(A=Multiply(A=@n15, B=@n16.RGBA), B=ScalarParameter() {defaultValue=0.5, parameterName=Background Opacity})) {constB=0.5}
  n18 = Multiply(A=NamedRerouteUsage()) {constB=0.2}
  n19 = NamedRerouteUsage()
  n20 = Constant4Vector() {constant=<rgba 1 0.1413 0 1>}
  Add(
   A=Add(A=Saturate(None=@n17), B=Saturate(None=@n18))
   B=Multiply(A=Multiply(A=@n19, B=@n20.RGBA), B=ScalarParameter() {defaultValue=0.7, parameterName=Triangle Opacity})
  )
  n21 = MaterialFunctionCall() {materialFunction=&MF_UI_9SliceUVs}
  n22 = Constant2Vector() {g=0.115, r=-0.39}
  n23 = MaterialFunctionCall(
   UVs=MaterialFunctionCall(UVs=Add(A=@n21.UVs, B=@n22), "Rotation Angle (0-1)"=Constant()) {materialFunction=&CustomRotator}.Rotated Values
   Size=Constant() {r=0.1}
  ) {materialFunction=&MF_UI_Box}
  n24 = MaterialFunctionCall() {materialFunction=&MF_UI_9SliceUVs}
  n25 = Constant2Vector() {r=-0.18}
  n26 = MaterialFunctionCall(UVs=MaterialFunctionCall(UVs=Add(A=@n24.UVs, B=@n25), "Rotation Angle (0-1)"=Constant() {r=0.124}) {materialFunction=&CustomRotator}.Rotated Values) {materialFunction=&MF_UI_Box}
  NamedRerouteDeclaration(None=Saturate(None=Subtract(A=@n23.Fill, B=@n26.Fill))) {name=Triangle, nodeColor=<rgba 1 0 0.9141 1>}
  TextureCoordinate()
  TextureCoordinate()
  TextureCoordinate()
 ]
}
