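# HUD message widget that announces "RIFT POD INCOMING" when a premium supply drop event arrives.
# It starts collapsed and listens on the global event router with an async task, stopped on Destruct.
# Each event shows it and plays Anim_Intro, a six-second sequence that animates icon and background material parameters, reveals text and arrows, then fades out.
# When the animation finishes, the widget collapses and broadcasts OnOutroAnimationDone.
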WidgetBlueprint /PremiumSupplyDrop/UI/WBP_DropPod_Incoming
animations [Anim_Intro {
  binding_count=6
  length_seconds=5.9941
  object=&self:Anim_Intro
  sequence {
   bindings [
    WBP_DropPod_Incoming_C_4 {
     id=#1
     possessed_class=WBP_DropPod_Incoming_C
     tracks [{
       type=MovieSceneFloatTrack
       display_name=Render Opacity
       sections [{
         type=MovieSceneFloatSection
         end_frame=179
         start_frame=0
         channels [None {type=MovieSceneScriptingFloatChannel, thinned_from=4, keys [[2,1],[168,1],[179.822,0]]}]
        }]
      }]
    }
    DROPPOD {
     id=#2
     possessed_class=UEFN_TextBlock_C
     tracks [
      {
       type=MovieSceneFloatTrack
       display_name=Render Opacity
       sections [{type=MovieSceneFloatSection, end_frame=36, start_frame=0, channels [None {type=MovieSceneScriptingFloatChannel, keys [[32,0],[36,1]]}]}]
      }
      {
       type=MovieScene2DTransformTrack
       sections [{
         type=MovieScene2DTransformSection
         end_frame=36
         start_frame=30
         channels [
          Translation.X {type=MovieSceneScriptingFloatChannel, keys [[30,-20],[36,0]]}
          Translation.Y {type=MovieSceneScriptingFloatChannel, default=0}
          Angle {type=MovieSceneScriptingFloatChannel, default=0}
          Scale.X {type=MovieSceneScriptingFloatChannel, default=1}
          Scale.Y {type=MovieSceneScriptingFloatChannel, default=1}
          Shear.X {type=MovieSceneScriptingFloatChannel, default=0}
          Shear.Y {type=MovieSceneScriptingFloatChannel, default=0}
         ]
        }]
      }
     ]
    }
    DropPodIcon {
     id=#3
     possessed_class=Image
     tracks [{
       type=MovieSceneWidgetMaterialTrack
       sections [{
         type=MovieSceneParameterSection
         end_frame=120
         start_frame=0
         channels [
          Outer Line Mask UV X {type=MovieSceneScriptingFloatChannel, keys [[2,0.3],[7,0.1]]}
          Outer Line Mask UV Y {type=MovieSceneScriptingFloatChannel, keys [[2,0.2],[7,0.3]]}
          Square 1 Opacity {type=MovieSceneScriptingFloatChannel, keys [[1,0],[2,0.7]]}
          Square 2 Opacity {type=MovieSceneScriptingFloatChannel, keys [[12,0],[13,0.7]]}
          Background Glow {type=MovieSceneScriptingFloatChannel, default=0.3, keys [[16,0.8],[22,0.3]]}
          Bottom Line Thickness {type=MovieSceneScriptingFloatChannel, keys [[13,0],[16,0.015]]}
          Icon Opacity {type=MovieSceneScriptingFloatChannel, keys [[24,0],[33,0.7]]}
          Background Glow Intensity {type=MovieSceneScriptingFloatChannel, keys [[24,2],[31,1],[52,1]]}
          Outer Line Mask SizeR {type=MovieSceneScriptingFloatChannel, keys [[2,0.25],[7,1]]}
          Outer Line Mask SizeG {type=MovieSceneScriptingFloatChannel, keys [[2,0.4],[7,0.2],[12,1.4]]}
          Outer Line Mask SizeB {type=MovieSceneScriptingFloatChannel, const [0,1]}
          Outer Line Mask SizeA {type=MovieSceneScriptingFloatChannel, const [1,1]}
         ]
        }]
      }]
    }
    Background {
     id=#4
     possessed_class=Image
     tracks [{
       type=MovieSceneWidgetMaterialTrack
       sections [{
         type=MovieSceneParameterSection
         end_frame=120
         start_frame=0
         channels [
          Triangle Opacity {type=MovieSceneScriptingFloatChannel, keys [[44,0],[51,1],[58,0],[65,1],[72,0],[79,1],[86,0],[93,1]]}
          Line Shape Mask UV X {type=MovieSceneScriptingFloatChannel, keys [[22,0.4],[25,0.55],[30,0]]}
          Line Shape Mask UV Y {type=MovieSceneScriptingFloatChannel, keys [[22,0.32],[25,0],[30,0]]}
          Background Opacity {type=MovieSceneScriptingFloatChannel, keys [[29,0],[32,0.5]]}
          Line Shape Mask SizeR {type=MovieSceneScriptingFloatChannel, keys [[22,0.2],[25,0.2],[30,1]]}
          Line Shape Mask SizeG {type=MovieSceneScriptingFloatChannel, keys [[22,0.2],[25,0.6],[30,1]]}
          Line Shape Mask SizeB {type=MovieSceneScriptingFloatChannel, const [0,1]}
          Line Shape Mask SizeA {type=MovieSceneScriptingFloatChannel, const [1,1]}
         ]
        }]
      }]
    }
    Incoming {
     id=#5
     possessed_class=UEFN_TextBlock_C
     tracks [
      {
       type=MovieSceneFloatTrack
       display_name=Render Opacity
       sections [{type=MovieSceneFloatSection, end_frame=39, start_frame=0, channels [None {type=MovieSceneScriptingFloatChannel, keys [[36,0],[39,1]]}]}]
      }
      {
       type=MovieScene2DTransformTrack
       sections [{
         type=MovieScene2DTransformSection
         end_frame=39
         start_frame=34
         channels [
          Translation.X {type=MovieSceneScriptingFloatChannel, keys [[34,-20],[39,0]]}
          Translation.Y {type=MovieSceneScriptingFloatChannel, default=0}
          Angle {type=MovieSceneScriptingFloatChannel, default=0}
          Scale.X {type=MovieSceneScriptingFloatChannel, default=1}
          Scale.Y {type=MovieSceneScriptingFloatChannel, default=1}
          Shear.X {type=MovieSceneScriptingFloatChannel, default=0}
          Shear.Y {type=MovieSceneScriptingFloatChannel, default=0}
         ]
        }]
      }
     ]
    }
    Arrows {
     id=#6
     possessed_class=Image
     tracks [{
       type=MovieSceneFloatTrack
       display_name=Render Opacity
       sections [{type=MovieSceneFloatSection, end_frame=41, start_frame=0, channels [None {type=MovieSceneScriptingFloatChannel, keys [[39,0],[41,1]]}]}]
      }]
    }
   ]
   display_rate {numerator=30}
  }
 }]
graphs [
 EventGraph {nodes [
   OnSupplySpawned() {is=title K2Node_CustomEvent}
   > PlayAnimation(InAnimation=Get Anim_Intro() -> (Anim_Intro:Widget Animation Object Reference))
   > SetVisibility(InVisibility=Visible)
   n1 = HideDropPodMessage() {is=title K2Node_CustomEvent}
   > SetVisibility(InVisibility=Collapsed)
   > Call On Outro Animation Done!(execute:Exec, self:WBP Drop Pod Incoming Object Reference) -> (then:Exec) {is=title K2Node_CallDelegate}
   Event OnInitialized()
   > BindToAnimationFinished(Animation=Get Anim_Intro(), Delegate=@n1)
   Event Construct()
   > SetVisibility(InVisibility=Collapsed)
   n2 = GetGlobalEventRouter()
   > n3 = StartListeningToEvent!(
    execute:Exec
    Target:Gameplay Event Router Component Object Reference=@n2
    EventType:Script Struct Object Reference=/PremiumSupplyDrop/Gameplay/PremiumSupplyDropEventData
    Context:Object Reference
    EventBubblingRule:EEventBubblingRule Enum
    None:%=@n2
   ) -> (
    then:Exec
    AsyncTaskProxy:Async Action Start Listening To Event Object Reference
    OnEventReceived:Exec
    Payload:Premium Supply Drop Event Data Structure
   )
   > Set WaitForPodSpawnTask(WaitForPodSpawnTask:Async Action Start Listening To Event Object Reference=@n3) -> (Output_Get:Async Action Start Listening To Event Object Reference)
   OnSupplySpawned(execute=@n3.OnEventReceived) {is=title K2Node_CallFunction}
   Event Destruct()
   > StopListeningToEvent(self:Async Action Start Listening To Event Object Reference=Get WaitForPodSpawnTask() -> (WaitForPodSpawnTask:Async Action Start Listening To Event Object Reference))
  ]}
 OnOutroAnimationDone {nodes [OnOutroAnimationDone() {is=title K2Node_FunctionEntry}]}
]
variables [
 OnOutroAnimationDone {}
 WaitForPodSpawnTask {type=AsyncAction_StartListeningToEvent, default=null}
]
widgets {
 info {inheritedWidgetCount=0, parentClass=&UserWidget, rootWidgetClass=&HorizontalBox, widgetCount=11}
 widgets [
  DropPodIncomingMessage {widgetClassPath=&HorizontalBox, widget=DropPodIncomingMessage}
  Overlay_54 {widgetClassPath=&Overlay, parent=DropPodIncomingMessage, widget=Overlay_54, slot {class=HorizontalBoxSlot}}
  DropPodIcon {
   widgetClassPath=&Image
   bIsVariable=true
   parent=Overlay_54
   widget=DropPodIcon
   properties {brush {
     bIsDynamicallyLoaded=false
     drawAs=Image
     imageType=FullColor
     mirroring=NoMirror
     resourceName=None
     resourceObject=/PremiumSupplyDrop/UI/Materials/M_UI_IconWithBackground
     tiling=NoTile
     imageSize=<150 150>
     margin=<ltrb 0 0 0 0>
     outlineSettings {
      bUseBrushTransparency=false
      roundingType=HalfHeightRadius
      width=0
      color {colorUseRule=UseColor_Specified, specifiedColor=<rgba 0 0 0 0>}
      cornerRadii=<0 0 0 0>
     }
     tintColor {colorUseRule=UseColor_Specified, specifiedColor=<rgba 1 1 1 1>}
     uVRegion {bIsValid=false, max=<0 0>, min=<0 0>}
    }}
   slot {class=OverlaySlot, properties {horizontalAlignment=HAlign_Center, verticalAlignment=VAlign_Center, padding=<ltrb -12 0 -7 0>}}
  }
  Overlay_87 {
   widgetClassPath=&Overlay
   parent=DropPodIncomingMessage
   widget=Overlay_87
   slot {class=HorizontalBoxSlot, properties {padding=<ltrb 0 30 0 30>}}
  }
  Background {
   widgetClassPath=&Image
   parent=Overlay_87
   widget=Background
   properties {brush {
     bIsDynamicallyLoaded=false
     drawAs=Image
     imageType=FullColor
     mirroring=NoMirror
     resourceName=None
     resourceObject=/PremiumSupplyDrop/UI/Materials/M_UI_TextBackground
     tiling=NoTile
     imageSize=<190 190>
     margin=<ltrb 0 0 0 0>
     outlineSettings {
      bUseBrushTransparency=false
      roundingType=HalfHeightRadius
      width=0
      color {colorUseRule=UseColor_Specified, specifiedColor=<rgba 0 0 0 0>}
      cornerRadii=<0 0 0 0>
     }
     tintColor {colorUseRule=UseColor_Specified, specifiedColor=<rgba 1 1 1 1>}
     uVRegion {bIsValid=false, max=<0 0>, min=<0 0>}
    }}
   slot {class=OverlaySlot, properties {horizontalAlignment=HAlign_Fill, verticalAlignment=VAlign_Fill, padding=<ltrb -10 -50 -20 -50>}}
  }
  VerticalBox_73 {
   widgetClassPath=&VerticalBox
   parent=Overlay_87
   widget=VerticalBox_73
   slot {class=OverlaySlot, properties {verticalAlignment=VAlign_Center, padding=<ltrb 17 1 10 0>}}
  }
  DROPPOD {
   widgetClassPath=&UEFN_TextBlock
   parent=VerticalBox_73
   widget=DROPPOD
   properties {
    text=RIFT POD
    font {
     bForceMonospaced=false
     bMaterialIsStencil=false
     fontMaterial=null
     fontObject=&Burbank
     letterSpacing=0
     monospacedWidth=1
     size=24
     skewAmount=0
     typefaceFontName=BurbankRegular-Black
     outlineSettings {
      bApplyOutlineToDropShadows=false
      bMiteredCorners=false
      bSeparateFillAlpha=false
      outlineMaterial=null
      outlineSize=0
      outlineColor=<rgba 0 0 0 1>
     }
    }
   }
   slot {class=VerticalBoxSlot, properties {padding=<ltrb 0 -5 0 0>}}
  }
  HorizontalBox_109 {
   widgetClassPath=&HorizontalBox
   parent=VerticalBox_73
   widget=HorizontalBox_109
   slot {class=VerticalBoxSlot, properties {horizontalAlignment=HAlign_Left, padding=<ltrb 0 -8 0 0>}}
  }
  Arrows {
   widgetClassPath=&Image
   bIsVariable=true
   parent=HorizontalBox_109
   widget=Arrows
   properties {brush {
     bIsDynamicallyLoaded=false
     drawAs=Image
     imageType=FullColor
     mirroring=NoMirror
     resourceName=None
     resourceObject=/PremiumSupplyDrop/UI/Materials/M_UI_Arrows
     tiling=NoTile
     imageSize=<32 10>
     margin=<ltrb 0 0 0 0>
     outlineSettings {
      bUseBrushTransparency=false
      roundingType=HalfHeightRadius
      width=0
      color {colorUseRule=UseColor_Specified, specifiedColor=<rgba 0 0 0 0>}
      cornerRadii=<0 0 0 0>
     }
     tintColor {colorUseRule=UseColor_Specified, specifiedColor=<rgba 1 1 1 1>}
     uVRegion {bIsValid=false, max=<0 0>, min=<0 0>}
    }}
   slot {class=HorizontalBoxSlot, properties {padding=<ltrb 0 5 0 7>}}
  }
  Spacer_82 {widgetClassPath=&Spacer, parent=HorizontalBox_109, widget=Spacer_82, properties {size=<6 1>}, slot {class=HorizontalBoxSlot}}
  Incoming {
   widgetClassPath=&UEFN_TextBlock
   parent=HorizontalBox_109
   widget=Incoming
   properties {
    text=INCOMING
    font {
     bForceMonospaced=false
     bMaterialIsStencil=false
     fontMaterial=null
     fontObject=&BurbankSmall
     letterSpacing=0
     monospacedWidth=1
     size=15.75
     skewAmount=0
     typefaceFontName=Bold
     outlineSettings {
      bApplyOutlineToDropShadows=false
      bMiteredCorners=false
      bSeparateFillAlpha=false
      outlineMaterial=null
      outlineSize=0
      outlineColor=<rgba 0 0 0 1>
     }
    }
   }
   slot {class=HorizontalBoxSlot, properties {horizontalAlignment=HAlign_Right, verticalAlignment=VAlign_Top}}
  }
 ]
}
