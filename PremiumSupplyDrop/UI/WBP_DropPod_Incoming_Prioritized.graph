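# Prioritized HUD widget that wraps the drop pod incoming notification for the premium supply drop.
# On construct it listens on the global event router for PremiumSupplyDropEventData and marks itself as having active content when one arrives.
# It also binds the inner WBP_DropPod_Incoming outro-animation-done event to clear active content.
# The listener task is stored and stopped on destruct.
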WidgetBlueprint /PremiumSupplyDrop/UI/WBP_DropPod_Incoming_Prioritized
graphs [EventGraph {nodes [
   n1 = HandleNotificationAnimComplete() {is=title K2Node_CustomEvent}
   > SetHasActiveContent!(execute, self:Fort Prioritized Widget Interface, bNewActive:bool=false) -> (then)
   Event Destruct()
   > StopListeningToEvent(self:Async Action Start Listening To Event Object Reference=Get WaitForPodSpawnTask() -> (WaitForPodSpawnTask:Async Action Start Listening To Event Object Reference))
   Event Construct()
   > n2 = Sequence()
   n3 = GetGlobalEventRouter()
   > n4 = StartListeningToEvent!(
    execute:Exec
    Target:Gameplay Event Router Component Object Reference=@n3
    EventType:Script Struct Object Reference=/PremiumSupplyDrop/Gameplay/PremiumSupplyDropEventData
    Context:Object Reference
    EventBubblingRule:EEventBubblingRule Enum
    None:%=@n3
   ) -> (
    then:Exec
    AsyncTaskProxy:Async Action Start Listening To Event Object Reference
    OnEventReceived:Exec
    Payload:Premium Supply Drop Event Data Structure
   )
   > Set WaitForPodSpawnTask(WaitForPodSpawnTask:Async Action Start Listening To Event Object Reference=@n4) -> (Output_Get:Async Action Start Listening To Event Object Reference)
   SetHasActiveContent(execute=@n4.OnEventReceived, bNewActive=true)
   Bind Event to On Outro Animation Done(
    execute=@n2.then_1
    self:WBP Drop Pod Incoming Object Reference=Get WBP_DropPod_Incoming() -> (WBP_DropPod_Incoming:WBP Drop Pod Incoming Object Reference)
    Delegate=@n1
   ) {is=title K2Node_AddDelegate}
  ]}]
variables [WaitForPodSpawnTask {type=AsyncAction_StartListeningToEvent, default=null}]
widgets {
 info {
  inheritedWidgetCount=1
  parentClass=/Script/FortniteUI.AthenaCommonPrioritizedWidget
  rootWidgetClass=&/WBP_DropPod_Incoming.WBP_DropPod_Incoming_C
  widgetCount=2
 }
 widgets [
  WBP_DropPod_Incoming {widgetClassPath=&/WBP_DropPod_Incoming, bIsVariable=true, widget=WBP_DropPod_Incoming}
  TutorialHighlight {widgetClassPath=&TutorialHighlight}
 ]
}
